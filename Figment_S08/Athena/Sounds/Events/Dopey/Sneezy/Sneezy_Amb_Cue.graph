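# Looping map-wide ambience cue for the Sneezy event in Figment S08, with layers that respond to a MeterLevel parameter.
# A non-spatialized Sneezy bed mixes with a spatialized, listener-focused, closer-range branch; a third attenuation branch has no input.
# The focused branch crossfades snow and cane loops across overlapping MeterLevel ranges, raising pitch as the level climbs.
# Stacked continuous modulators fade some layers in then down, and all attenuations use occlusion.
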SoundCue /Figment_S08/Athena/Sounds/Events/Dopey/Sneezy/Sneezy_Amb_Cue
root=Mixer_0
nodes [
 {
  id=Attenuation_0
  children [WavePlayer_2]
  properties {
   bOverrideAttenuation=true
   attenuationOverrides {
    bEnableOcclusion=true
    bSpatialize=false
    falloffDistance=4000
    occlusionLowPassFilterFrequency=1000
    occlusionTraceChannel=ECC_GameTraceChannel11
    attenuationShapeExtents {x=0}
   }
  }
 }
 {
  id=Attenuation_1
  children [null]
  properties {
   bOverrideAttenuation=true
   attenuationOverrides {
    bEnableFocusInterpolation=true
    bEnableListenerFocus=true
    bEnableOcclusion=true
    falloffDistance=4000
    focusAttackInterpSpeed=0.2
    focusAzimuth=2
    nonFocusAzimuth=10
    nonFocusVolumeAttenuation=0.5
    occlusionLowPassFilterFrequency=1000
    occlusionTraceChannel=ECC_GameTraceChannel11
    stereoSpread=1024
    attenuationShapeExtents {x=0}
   }
  }
 }
 {
  id=Attenuation_2
  children [Mixer_1]
  properties {
   bOverrideAttenuation=true
   attenuationOverrides {
    bEnableFocusInterpolation=true
    bEnableListenerFocus=true
    bEnableOcclusion=true
    falloffDistance=2000
    focusAttackInterpSpeed=5
    focusAzimuth=10
    focusReleaseInterpSpeed=0.2
    nonFocusAzimuth=45
    nonFocusVolumeAttenuation=0.5
    occlusionLowPassFilterFrequency=1000
    occlusionTraceChannel=ECC_GameTraceChannel11
    stereoSpread=1024
    attenuationShapeExtents {x=2000}
   }
  }
 }
 {
  id=Enveloper_0
  children [ModulatorContinuous_4]
  properties {pitchCurve {editorCurveData {keys [{time=0}]}}, volumeCurve {editorCurveData {keys [<at 0 1>]}}}
 }
 {id=Mixer_0, children [Attenuation_0, Attenuation_1, Attenuation_2], properties {inputVolume [1,2,1.5]}}
 {
  id=Mixer_1
  children [Enveloper_0, ModulatorContinuous_7, ModulatorContinuous_2, ModulatorContinuous_5]
  properties {inputVolume [1,1,1,1.5]}
 }
 {
  id=ModulatorContinuous_0
  children [WavePlayer_0]
  properties {
   pitchModulationParams {default=0, maxOutput=1.25, minOutput=0.9, parameterName=MeterLevel}
   volumeModulationParams {default=0, maxInput=0.4, minInput=0.1, parameterName=MeterLevel}
  }
 }
 {
  id=ModulatorContinuous_2
  children [WavePlayer_4]
  properties {
   pitchModulationParams {default=0, maxOutput=1.25, minOutput=0.8, parameterName=MeterLevel}
   volumeModulationParams {default=0, maxInput=0.8, minInput=0.4, minOutput=0.1, parameterName=MeterLevel}
  }
 }
 {
  id=ModulatorContinuous_3
  children [WavePlayer_5]
  properties {
   pitchModulationParams {default=0, maxOutput=1.5, minOutput=0.9, parameterName=MeterLevel}
   volumeModulationParams {default=0, maxInput=0.6, minInput=0.2, parameterName=MeterLevel}
  }
 }
 {
  id=ModulatorContinuous_4
  children [ModulatorContinuous_0]
  properties {
   pitchModulationParams {default=0, maxOutput=1.2, minInput=0.8, minOutput=1}
   volumeModulationParams {default=0, maxOutput=0.5, minInput=0.7, minOutput=1, parameterName=MeterLevel}
  }
 }
 {
  id=ModulatorContinuous_5
  children [WavePlayer_3]
  properties {
   pitchModulationParams {default=0, maxOutput=1.25, minOutput=0.8, parameterName=MeterLevel}
   volumeModulationParams {default=0, minInput=0.8, parameterName=MeterLevel}
  }
 }
 {
  id=ModulatorContinuous_7
  children [ModulatorContinuous_3]
  properties {
   pitchModulationParams {default=0, minOutput=1}
   volumeModulationParams {default=0, maxOutput=0, minInput=0.8, minOutput=1, parameterName=MeterLevel}
  }
 }
 {id=WavePlayer_0, properties {bLooping=true, soundWaveAssetPtr=/Figment_S08/Athena/Sounds/Events/Dopey/Snow/Snow_Ambient_01}}
 {id=WavePlayer_2, properties {bLooping=true, soundWaveAssetPtr=&/Sneezy_Amb_01}}
 {id=WavePlayer_3, properties {bLooping=true, soundWaveAssetPtr=/Figment_S08/Athena/Sounds/Events/Cane/Cane_Loop_01}}
 {id=WavePlayer_4, properties {bLooping=true, soundWaveAssetPtr=/Figment_S08/Athena/Sounds/Events/Dopey/Snow/Snow_Dest_Amb_02_LOOP}}
 {id=WavePlayer_5, properties {bLooping=true, soundWaveAssetPtr=/Figment_S08/Athena/Sounds/Events/Dopey/Snow/Snow_Dest_Amb_03_LOOP}}
]
properties {
 maxDistance=4000
 soundClassObject=/Figment_S08/Athena/Sounds/Events/SoundClasses/Fort_Events_Mapwide
 volumeMultiplier=2
 attenuationOverrides {
  bEnableOcclusion=true
  bSpatialize=false
  falloffDistance=4000
  occlusionInterpolationTime=1
  occlusionLowPassFilterFrequency=1000
  occlusionTraceChannel=ECC_GameTraceChannel11
  stereoSpread=1024
  attenuationShapeExtents {x=0}
 }
}
