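; Sneezy_MusicParty is an event blueprint that switches the music party off and on around the GW3 calendar event.
; On begin play it calls an async authority match-ready action and watches the GW3 calendar event.
; When GW3 becomes active it stops the music party with an argument of 2.0, and when GW3 ends it starts the music party again.
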(event EventBeginPlay
  (Class|FortAsyncActionAuthorityMatchReady|MatchReadyAsync)
  (MCPContext|WatchCalendarEvent "GW3"
    (:OnEventActive
      (Class|MusicPartyControllerInterface|StopMusicParty 2.0))
    (:OnEventEnded
      (Class|MusicPartyControllerInterface|StartMusicParty))))

(fn ConstructionScript)
