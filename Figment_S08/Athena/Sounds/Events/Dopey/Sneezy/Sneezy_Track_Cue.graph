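# Looping Sound Cue for the Sneezy music track in the Dopey event audio, assigned to a quest dance-music sound class.
# A Mixer blends two looping Sneezy_Track_01 players, each with its own attenuation override.
# The full-level layer is non-spatialized with a shorter falloff.
# The quieter layer is spatialized and occluded, using a custom curve that peaks at mid-range and fades out nearby and far away.
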SoundCue /Figment_S08/Athena/Sounds/Events/Dopey/Sneezy/Sneezy_Track_Cue
root=Mixer_0
nodes [
 {
  id=Attenuation_0
  children [WavePlayer_2]
  properties {bOverrideAttenuation=true, attenuationOverrides {bSpatialize=false, falloffDistance=4000, stereoSpread=512, attenuationShapeExtents {x=0}}}
 }
 {
  id=Attenuation_1
  children [WavePlayer_0]
  properties {
   bOverrideAttenuation=true
   attenuationOverrides {
    bEnableOcclusion=true
    distanceAlgorithm=Custom
    falloffDistance=10000
    occlusionLowPassFilterFrequency=1000
    occlusionTraceChannel=ECC_GameTraceChannel11
    stereoSpread=1024
    attenuationShapeExtents {x=0}
    customAttenuationCurve {editorCurveData {keys [<at 0 0>,<at 0.4 1>,<at 1 0>]}}
   }
  }
 }
 {id=Mixer_0, children [Attenuation_0, Attenuation_1], properties {inputVolume [1,0.6]}}
 {id=WavePlayer_0, properties {bLooping=true, soundWaveAssetPtr=&/Sneezy_Track_01}}
 {id=WavePlayer_2, properties {bLooping=true, soundWaveAssetPtr=&Sneezy_Track_01}}
]
properties {
 maxDistance=10000
 soundClassObject=/Game/Sounds/FortSoundClassesAndModes/Fort_DanceMusic_Quest_PSM
 volumeMultiplier=2
 attenuationOverrides {
  bEnableOcclusion=true
  bSpatialize=false
  falloffDistance=4000
  occlusionInterpolationTime=1
  occlusionLowPassFilterFrequency=1000
  occlusionTraceChannel=ECC_GameTraceChannel11
  stereoSpread=1024
  attenuationShapeExtents {x=0}
 }
}
