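# Sound cue for the end of the Sneezy track in the Dopey event sounds, playing a power-down sound borrowed from the Leaky set.
# A single wave player feeds a Mixer whose second input is empty, at half volume on the dance music quest sound class.
# It references the dance floor music attenuation but overrides it locally, enabling occlusion with a low-pass filter on a custom trace channel.
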SoundCue /Figment_S08/Athena/Sounds/Events/Dopey/Sneezy/Sneezy_Track_End_Cue
root=Mixer_0
nodes [
 {id=Mixer_0, children [WavePlayer_0,null], properties {inputVolume [1,1]}}
 {id=WavePlayer_0, properties {soundWaveAssetPtr=/Figment_S08/Athena/Sounds/Events/Dopey/Leaky/Leaky_Power_Down_01}}
]
properties {
 attenuationSettings=/Game/Sounds/Attenuation/Fort_DanceFloorMusic_Athena_Att
 bOverrideAttenuation=true
 maxDistance=3600
 soundClassObject=/Game/Sounds/FortSoundClassesAndModes/Fort_DanceMusic_Quest_PSM
 volumeMultiplier=0.5
 attenuationOverrides {
  bEnableOcclusion=true
  occlusionInterpolationTime=1
  occlusionLowPassFilterFrequency=1000
  occlusionTraceChannel=ECC_GameTraceChannel11
  stereoSpread=1024
  attenuationShapeExtents {x=0}
 }
}
