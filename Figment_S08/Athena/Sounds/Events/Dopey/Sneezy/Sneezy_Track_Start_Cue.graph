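# Sound cue for the start of a Sneezy track in the Figment S08 Dopey event audio.
# A mixer layers the Leaky power-up wave, borrowed from the sibling Leaky folder, with one of four Sneezy vocals picked at random with equal weight.
# It uses a dance-music quest sound class and references dance-floor music attenuation, but enables an attenuation override with occlusion low-pass filtering and wide stereo spread.
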SoundCue /Figment_S08/Athena/Sounds/Events/Dopey/Sneezy/Sneezy_Track_Start_Cue
root=Mixer_0
nodes [
 {id=Mixer_0, children [WavePlayer_0,Random_0], properties {inputVolume [1,1]}}
 {id=Random_0, children [WavePlayer_1,WavePlayer_2,WavePlayer_3,WavePlayer_4], properties {weights [1,1,1,1]}}
 {id=WavePlayer_0, properties {soundWaveAssetPtr=/Figment_S08/Athena/Sounds/Events/Dopey/Leaky/Leaky_Power_Up_01}}
 {id=WavePlayer_1, properties {soundWaveAssetPtr=&/Sneezy_Vocal_01}}
 {id=WavePlayer_2, properties {soundWaveAssetPtr=&/Sneezy_Vocal_03}}
 {id=WavePlayer_3, properties {soundWaveAssetPtr=&/Sneezy_Vocal_02}}
 {id=WavePlayer_4, properties {soundWaveAssetPtr=&/Sneezy_Vocal_04}}
]
properties {
 attenuationSettings=/Game/Sounds/Attenuation/Fort_DanceFloorMusic_Athena_Att
 bOverrideAttenuation=true
 maxDistance=3600
 soundClassObject=/Game/Sounds/FortSoundClassesAndModes/Fort_DanceMusic_Quest_PSM
 volumeMultiplier=1
 attenuationOverrides {
  bEnableOcclusion=true
  occlusionInterpolationTime=1
  occlusionLowPassFilterFrequency=1000
  occlusionTraceChannel=ECC_GameTraceChannel11
  stereoSpread=1024
  attenuationShapeExtents {x=0}
 }
}
